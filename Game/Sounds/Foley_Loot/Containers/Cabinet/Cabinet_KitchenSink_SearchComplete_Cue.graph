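# Sound cue for the end of searching a kitchen sink cabinet loot container, in the loot collect sound class.
# A quality-level root sends the first two levels to a wave with slight random pitch and volume modulation.
# The last two levels play a separate "opt" wave unmodulated.
# Playback uses prop search attenuation and overrides concurrency to cap simultaneous instances, stopping the farthest first.
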SoundCue /Game/Sounds/Foley_Loot/Containers/Cabinet/Cabinet_KitchenSink_SearchComplete_Cue
root=QualityLevel_0
nodes [
 {id=Modulator_1, children [WavePlayer_10], properties {pitchMax=1.07, pitchMin=0.92, volumeMax=1.07, volumeMin=0.92}}
 {id=QualityLevel_0, children [Modulator_1,Modulator_1,WavePlayer_0,WavePlayer_0]}
 {id=WavePlayer_0, properties {soundWaveAssetPtr=&/Cabinet_KitchenSink_SearchEnd_opt}}
 {id=WavePlayer_10, properties {soundWaveAssetPtr=&/Cabinet_KitchenSink_SearchEnd4}}
]
properties {
 attenuationSettings=/Game/Sounds/Attenuation/Fort_Prop_Search_Att
 bOverrideConcurrency=true
 maxDistance=1300
 soundClassObject=&Fort_Loot_Collect
 volumeMultiplier=0.7
 concurrencyOverrides {maxCount=4, resolutionRule=StopFarthestThenPreventNew}
}
