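# Sound cue for searching a kitchen sink cabinet loot container, routed through the loot-collect sound class with prop-search attenuation.
# A quality-level root selects among four branches: two layered mixes with three or two random variants per layer, and two using a single optimized wave.
# The mixes place quiet, pitch- and volume-modulated cabinet sounds under louder generic gear rummaging; the single-wave branches are modulated too.
# A concurrency override blocks new instances at the limit.
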SoundCue /Game/Sounds/Foley_Loot/Containers/Cabinet/Cabinet_KitchenSink_Search_Cue
root=QualityLevel_0
nodes [
 {id=Mixer_0, children [Modulator_0,Random_1], properties {inputVolume [0.2,1.2]}}
 {id=Mixer_17, children [Modulator_24,Random_16], properties {inputVolume [0.2,1.2]}}
 {id=Modulator_0, children [Random_0], properties {pitchMax=1.07, pitchMin=0.92, volumeMax=1.07, volumeMin=0.92}}
 {id=Modulator_1, children [WavePlayer_0], properties {pitchMax=1.07, pitchMin=0.92, volumeMax=1.07, volumeMin=0.92}}
 {id=Modulator_2, children [WavePlayer_5], properties {pitchMax=1.07, pitchMin=0.92, volumeMax=1.07, volumeMin=0.92}}
 {id=Modulator_24, children [Random_2], properties {pitchMax=1.07, pitchMin=0.92, volumeMax=1.07, volumeMin=0.92}}
 {id=QualityLevel_0, children [Mixer_17,Mixer_0,Modulator_1,Modulator_2]}
 {id=Random_0, children [WavePlayer_1,WavePlayer_2], properties {weights [1,1]}}
 {id=Random_1, children [WavePlayer_3,WavePlayer_4], properties {weights [1,1]}}
 {id=Random_16, children [WavePlayer_84, WavePlayer_85, WavePlayer_87], properties {weights [1,1,1]}}
 {id=Random_2, children [WavePlayer_12, WavePlayer_14, WavePlayer_15], properties {weights [1,1,1]}}
 {id=WavePlayer_0, properties {soundWaveAssetPtr=&/Cabinet_KitchenSink_Search_opt}}
 {id=WavePlayer_1, properties {soundWaveAssetPtr=&/Cabinet_KitchenSink_Search2}}
 {id=WavePlayer_12, properties {soundWaveAssetPtr=&Cabinet_KitchenSink_Search2}}
 {id=WavePlayer_14, properties {soundWaveAssetPtr=&/Cabinet_KitchenSink_Search4}}
 {id=WavePlayer_15, properties {soundWaveAssetPtr=&/Cabinet_KitchenSink_Search5}}
 {id=WavePlayer_2, properties {soundWaveAssetPtr=&Cabinet_KitchenSink_Search4}}
 {id=WavePlayer_3, properties {soundWaveAssetPtr=/Game/Sounds/Foley_Loot/Containers/Misc/Prop_Gear_Search_05}}
 {id=WavePlayer_4, properties {soundWaveAssetPtr=/Game/Sounds/Foley_Loot/Containers/Misc/Prop_Gear_Search_06}}
 {id=WavePlayer_5, properties {soundWaveAssetPtr=&Cabinet_KitchenSink_Search_opt}}
 {id=WavePlayer_84, properties {soundWaveAssetPtr=&Prop_Gear_Search_05}}
 {id=WavePlayer_85, properties {soundWaveAssetPtr=&Prop_Gear_Search_06}}
 {id=WavePlayer_87, properties {soundWaveAssetPtr=/Game/Sounds/Foley_Loot/Containers/Misc/Prop_Gear_Search_01}}
]
properties {
 attenuationSettings=/Game/Sounds/Attenuation/Fort_Prop_Search_Att
 bOverrideConcurrency=true
 maxDistance=1300
 soundClassObject=&Fort_Loot_Collect
 volumeMultiplier=1.1
 concurrencyOverrides {resolutionRule=PreventNew}
}
